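# Physics asset for the left-leg shorts of the medium male FavorDockLargo cosmetic body, apparently a backup by its name.
# Front, back, inner and outer chains each hang two capsule bones from a small MSTR anchor sphere, with locked twist and wider swing on the lower segment.
# Free-rotation constraints join the four lower capsules into a ring around the leg.
# An unconstrained thigh twist capsule provides the leg body.
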PhysicsAsset /BRCosmetics/Characters/Player/Male/Medium/Bodies/M_MED_FavorDockLargo/Meshes/M_MED_FavorDockLargo_PhysicsLegLBackUp
bodies {
 MSTR_shorts_bk_l [{shapeType=Sphere, radius=0.505, shapeName=MSTR_shorts_bk_l_sphere}]
 MSTR_shorts_fr_l [{shapeType=Sphere, radius=0.505, shapeName=MSTR_shorts_fr_l_sphere}]
 MSTR_shorts_in_l [{shapeType=Sphere, radius=0.505, shapeName=MSTR_shorts_in_l_sphere}]
 MSTR_shorts_out_l [{shapeType=Sphere, radius=0.505, shapeName=MSTR_shorts_out_l_sphere}]
 dyn_shorts_bk_01_l [{
   shapeType=Capsule
   length=6.5679
   radius=3.3809
   shapeName=dyn_shorts_bk_02_l_capsule
   center=<1.6897 -0.2731 0.99>
   rotation=<rot 2.8862 -90 5.057>
  }]
 dyn_shorts_bk_02_l [{
   shapeType=Capsule
   length=6.5679
   radius=3.3809
   shapeName=dyn_shorts_bk_02_l_capsule
   center=<1.6897 0.6107 0.9454>
   rotation=<rot 2.8862 -90 5.057>
  }]
 dyn_shorts_fr_01_l [{
   shapeType=Capsule
   length=5.1888
   radius=4.1082
   shapeName=dyn_shorts_fr_02_l_capsule
   center=<1.3677 -0.9679 2.3851>
   rotation=<rot 0.5821 -90.2587 -0.8306>
  }]
 dyn_shorts_fr_02_l [{
   shapeType=Capsule
   length=5.1888
   radius=4.1082
   shapeName=dyn_shorts_fr_02_l_capsule
   center=<1.3662 -1.1871 2.4167>
   rotation=<rot 2.073 -90.2803 -0.8311>
  }]
 dyn_shorts_in_01_l [{
   shapeType=Capsule
   length=7.9981
   radius=2.7983
   shapeName=dyn_shorts_in_02_l_capsule
   center=<2.4106 -0.0941 -1.62>
   rotation=<rot -0.7972 -91.1313 -6.0143>
  }]
 dyn_shorts_in_02_l [{
   shapeType=Capsule
   length=7.9981
   radius=2.7983
   shapeName=dyn_shorts_in_02_l_capsule
   center=<2.1846 0.0875 0.5613>
   rotation=<rot -1.7073 -91.0354 -6.0164>
  }]
 dyn_shorts_out_01_l [{
   shapeType=Capsule
   length=5.8523
   radius=3.2714
   shapeName=dyn_shorts_out_02_l_capsule
   center=<2.6775 -0.7633 0.5518>
   rotation=<rot 2.5505 -90 -2.979>
  }]
 dyn_shorts_out_02_l [{
   shapeType=Capsule
   length=5.8523
   radius=3.2714
   shapeName=dyn_shorts_out_02_l_capsule
   center=<2.8535 -1.3526 0.5606>
   rotation=<rot 2.5505 -90 0.389>
  }]
 thigh_twist_01_l [{
   shapeType=Capsule
   length=22.1122
   radius=5.3631
   shapeName=thigh_twist_01_l_capsule
   center=<-3.1801 -0.0583 0.281>
   rotation=<rot -89.5573 -179.9986 179.9983>
  }]
}
constraints [
 {
  bone1Name=dyn_shorts_bk_01_l
  bone2Name=MSTR_shorts_bk_l
  swing1LimitDegrees=15
  swing2LimitDegrees=10
  twistLimitDegrees=45
  twistMotion=Locked
 }
 {
  bone1Name=dyn_shorts_bk_02_l
  bone2Name=dyn_shorts_bk_01_l
  swing1LimitDegrees=30
  swing2LimitDegrees=10
  twistLimitDegrees=45
  twistMotion=Locked
 }
 {
  bone1Name=dyn_shorts_fr_01_l
  bone2Name=MSTR_shorts_fr_l
  swing1LimitDegrees=15
  swing2LimitDegrees=10
  twistLimitDegrees=45
  twistMotion=Locked
 }
 {
  bone1Name=dyn_shorts_fr_02_l
  bone2Name=dyn_shorts_fr_01_l
  swing1LimitDegrees=30
  swing2LimitDegrees=10
  twistLimitDegrees=45
  twistMotion=Locked
 }
 {
  bone1Name=dyn_shorts_in_01_l
  bone2Name=MSTR_shorts_in_l
  swing1LimitDegrees=15
  swing2LimitDegrees=10
  twistLimitDegrees=45
  twistMotion=Locked
 }
 {
  bone1Name=dyn_shorts_in_02_l
  bone2Name=dyn_shorts_in_01_l
  swing1LimitDegrees=30
  swing2LimitDegrees=10
  twistLimitDegrees=45
  twistMotion=Locked
 }
 {
  bone1Name=dyn_shorts_out_01_l
  bone2Name=MSTR_shorts_out_l
  swing1LimitDegrees=15
  swing2LimitDegrees=10
  twistLimitDegrees=45
  twistMotion=Locked
 }
 {
  bone1Name=dyn_shorts_out_02_l
  bone2Name=dyn_shorts_out_01_l
  swing1LimitDegrees=30
  swing2LimitDegrees=10
  twistLimitDegrees=45
  twistMotion=Locked
 }
 {
  bone1Name=dyn_shorts_out_02_l
  bone2Name=dyn_shorts_bk_02_l
  swing1LimitDegrees=45
  swing1Motion=Free
  swing2LimitDegrees=45
  swing2Motion=Free
  twistLimitDegrees=45
  twistMotion=Free
 }
 {
  bone1Name=dyn_shorts_in_02_l
  bone2Name=dyn_shorts_bk_02_l
  swing1LimitDegrees=45
  swing1Motion=Free
  swing2LimitDegrees=45
  swing2Motion=Free
  twistLimitDegrees=45
  twistMotion=Free
 }
 {
  bone1Name=dyn_shorts_fr_02_l
  bone2Name=dyn_shorts_in_02_l
  swing1LimitDegrees=45
  swing1Motion=Free
  swing2LimitDegrees=45
  swing2Motion=Free
  twistLimitDegrees=45
  twistMotion=Free
 }
 {
  bone1Name=dyn_shorts_fr_02_l
  bone2Name=dyn_shorts_out_02_l
  swing1LimitDegrees=45
  swing1Motion=Free
  swing2LimitDegrees=45
  swing2Motion=Free
  twistLimitDegrees=45
  twistMotion=Free
 }
]
