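# Post-copy AnimBP for the FavorDockLargo male body cosmetic, adding secondary motion to shorts and chains.
# It copies the attached parent's pose, runs a Control Rig, then chains Anim Dynamics springs for both shorts legs.
# Per-leg alphas, updated from calf-to-thigh rotation deltas, fade shorts dynamics out at deep knee bends.
# Weighted constraints and spine_05 look-ats position chain roots before their springs; a RigidBody node is unconnected.
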AnimBlueprint /BRCosmetics/Characters/Player/Male/Medium/Bodies/M_MED_FavorDockLargo/Meshes/M_MED_FavorDockLargo_AnimBP
target_skeleton=&/M_MED_FavorDockLargo_Skeleton
graphs [
 AnimGraph {nodes [
   RigidBody() {
    bClampLinearTranslationLimitToRefPose=true
    bOverrideWorldGravity=true
    overridePhysicsAsset=&/M_MED_FavorDockLargo_PhysicsLegLBackUp
    simulationSpace=BaseBoneSpace
    baseBoneRef {boneName=thigh_l}
    overrideWorldGravity {z=-980}
    simSpaceSettings {
     dampingAlpha=0
     maxAngularAcceleration=150
     maxAngularVelocity=30
     maxLinearAcceleration=150
     maxLinearVelocity=7
     velocityScaleZ=0.7
     worldAlpha=1
     externalLinearDragV {z=0.5}
    }
   }
   n1 = Control Rig(Source=Copy Pose From Mesh() {bUseAttachedParent=true}) {controlRigAssetReference {blueprintRigClass=&/M_MED_FavorDockLargo_CtrlRig.M_MED_FavorDockLargo_CtrlRig_C}}
   n2 = Get KneeBendL() -> (KneeBendL:float)
   n3 = Anim Dynamics - Bone: dyn_shorts_bk_01_l(ComponentPose=Local To Component(LocalPose=@n1), Alpha=@n2, ExternalForce="0,0,0") {
    angularSpringConstant=50
    bAngularSpring=true
    bUseGravityOverride=true
    simulationSpace=RootRelative
    boundBone {boneName=dyn_shorts_bk_01_l}
    componentAppliedLinearAccClamp=<1500 1500 1500>
    componentLinearAccScale=<0.2 0.2 0.3>
    componentLinearVelScale=<0.2 0.2 0.4>
    gravityOverride {z=-980}
    physicsBodyDefinitions [{
      boundBone {boneName=dyn_shorts_bk_01_l}
      boxExtents=<16 16 16>
      constraintSetup {angularLimitsMax=<0 5 5>, angularLimitsMin=<0 -5 -30>, angularTarget=<0 0 0>}
      localJointOffset=<1 0 0>
     }]
    simSpaceSettings {maxAngularAcceleration=15, maxAngularVelocity=3, simSpaceAngularAlpha=0.1}
   }
   n4 = Anim Dynamics - Bone: dyn_shorts_bk_02_l(ComponentPose=@n3, Alpha=@n2, ExternalForce="0,0,0") {
    angularSpringConstant=50
    bAngularSpring=true
    bUseGravityOverride=true
    simulationSpace=RootRelative
    boundBone {boneName=dyn_shorts_bk_02_l}
    componentAppliedLinearAccClamp=<1500 1500 1500>
    componentLinearAccScale=<0.2 0.2 0.3>
    componentLinearVelScale=<0.2 0.2 0.4>
    gravityOverride {z=-980}
    physicsBodyDefinitions [{
      boundBone {boneName=dyn_shorts_bk_02_l}
      boxExtents=<16 16 16>
      constraintSetup {angularLimitsMax=<0 5 0>, angularLimitsMin=<0 -5 -30>, angularTarget=<0 0 0>}
      localJointOffset=<1 0 0>
     }]
    simSpaceSettings {maxAngularAcceleration=15, maxAngularVelocity=3, simSpaceAngularAlpha=0.1}
   }
   n5 = Anim Dynamics - Bone: dyn_shorts_out_01_l(ComponentPose=@n4, Alpha=@n2, ExternalForce="0,0,0") {
    angularSpringConstant=50
    bAngularSpring=true
    bUseGravityOverride=true
    simulationSpace=RootRelative
    boundBone {boneName=dyn_shorts_out_01_l}
    componentAppliedLinearAccClamp=<1500 1500 1500>
    componentLinearAccScale=<0.2 0.2 0.3>
    componentLinearVelScale=<0.2 0.2 0.4>
    gravityOverride {z=-980}
    physicsBodyDefinitions [{
      boundBone {boneName=dyn_shorts_out_01_l}
      boxExtents=<16 16 16>
      constraintSetup {angularLimitsMax=<0 5 2>, angularLimitsMin=<0 -5 -15>, angularTarget=<0 0 0>}
      localJointOffset=<1 0 0>
     }]
    simSpaceSettings {maxAngularAcceleration=15, maxAngularVelocity=3, simSpaceAngularAlpha=0.1}
   }
   n6 = Anim Dynamics - Bone: dyn_shorts_out_02_l(ComponentPose=@n5, Alpha=@n2, ExternalForce="0,0,0") {
    angularSpringConstant=50
    bAngularSpring=true
    bUseGravityOverride=true
    simulationSpace=RootRelative
    boundBone {boneName=dyn_shorts_out_02_l}
    componentAppliedLinearAccClamp=<1500 1500 1500>
    componentLinearAccScale=<0.2 0.2 0.3>
    componentLinearVelScale=<0.2 0.2 0.4>
    gravityOverride {z=-980}
    physicsBodyDefinitions [{
      boundBone {boneName=dyn_shorts_out_02_l}
      boxExtents=<16 16 16>
      constraintSetup {angularLimitsMax=<0 5 2>, angularLimitsMin=<0 -5 -15>, angularTarget=<0 0 0>}
      localJointOffset=<1 0 0>
     }]
    simSpaceSettings {maxAngularAcceleration=15, maxAngularVelocity=3, simSpaceAngularAlpha=0.1}
   }
   n7 = Anim Dynamics - Bone: dyn_shorts_in_01_l(ComponentPose=@n6, Alpha=@n2, ExternalForce="0,0,0") {
    angularSpringConstant=50
    bAngularSpring=true
    bUseGravityOverride=true
    simulationSpace=RootRelative
    boundBone {boneName=dyn_shorts_in_01_l}
    componentAppliedLinearAccClamp=<1500 1500 1500>
    componentLinearAccScale=<0.2 0.2 0.3>
    componentLinearVelScale=<0.2 0.2 0.4>
    gravityOverride {z=-980}
    physicsBodyDefinitions [{
      boundBone {boneName=dyn_shorts_in_01_l}
      boxExtents=<16 16 16>
      constraintSetup {angularLimitsMax=<0 5 2>, angularLimitsMin=<0 -5 -5>, angularTarget=<0 0 0>}
      localJointOffset=<1 0 0>
     }]
    simSpaceSettings {maxAngularAcceleration=15, maxAngularVelocity=3, simSpaceAngularAlpha=0.1}
   }
   n8 = Anim Dynamics - Bone: dyn_shorts_in_02_l(ComponentPose=@n7, Alpha=@n2, ExternalForce="0,0,0") {
    angularSpringConstant=50
    bAngularSpring=true
    bUseGravityOverride=true
    simulationSpace=RootRelative
    boundBone {boneName=dyn_shorts_in_02_l}
    componentAppliedLinearAccClamp=<1500 1500 1500>
    componentLinearAccScale=<0.2 0.2 0.3>
    componentLinearVelScale=<0.2 0.2 0.4>
    gravityOverride {z=-980}
    physicsBodyDefinitions [{
      boundBone {boneName=dyn_shorts_in_02_l}
      boxExtents=<16 16 16>
      constraintSetup {angularLimitsMax=<0 5 2>, angularLimitsMin=<0 -5 -7>, angularTarget=<0 0 0>}
      localJointOffset=<1 0 0>
     }]
    simSpaceSettings {maxAngularAcceleration=15, maxAngularVelocity=3, simSpaceAngularAlpha=0.1}
   }
   n9 = Anim Dynamics - Bone: dyn_shorts_fr_01_l(ComponentPose=@n8, Alpha=@n2, ExternalForce="0,0,0") {
    angularSpringConstant=50
    bAngularSpring=true
    bUseGravityOverride=true
    simulationSpace=RootRelative
    boundBone {boneName=dyn_shorts_fr_01_l}
    componentAppliedLinearAccClamp=<1500 1500 1500>
    componentLinearAccScale=<0.2 0.2 0.3>
    componentLinearVelScale=<0.2 0.2 0.4>
    gravityOverride {z=-980}
    physicsBodyDefinitions [{
      boundBone {boneName=dyn_shorts_fr_01_l}
      boxExtents=<16 16 16>
      constraintSetup {angularLimitsMax=<0 5 5>, angularLimitsMin=<0 -5 -5>, angularTarget=<0 0 0>}
      localJointOffset=<1 0 0>
     }]
    simSpaceSettings {maxAngularAcceleration=15, maxAngularVelocity=3, simSpaceAngularAlpha=0.1}
   }
   n10 = Anim Dynamics - Bone: dyn_shorts_fr_02_l(ComponentPose=@n9, Alpha=@n2, ExternalForce="0,0,0") {
    angularSpringConstant=50
    bAngularSpring=true
    bUseGravityOverride=true
    simulationSpace=RootRelative
    boundBone {boneName=dyn_shorts_fr_02_l}
    componentAppliedLinearAccClamp=<1500 1500 1500>
    componentLinearAccScale=<0.2 0.2 0.3>
    componentLinearVelScale=<0.2 0.2 0.4>
    gravityOverride {z=-980}
    physicsBodyDefinitions [{
      boundBone {boneName=dyn_shorts_fr_02_l}
      boxExtents=<16 16 16>
      constraintSetup {angularLimitsMax=<0 5 5>, angularLimitsMin=<0 -5 -5>, angularTarget=<0 0 0>}
      localJointOffset=<1 0 0>
     }]
    simSpaceSettings {maxAngularAcceleration=15, maxAngularVelocity=3, simSpaceAngularAlpha=0.1}
   }
   n11 = Get KneeBendR() -> (KneeBendR:float)
   n12 = Anim Dynamics - Bone: dyn_shorts_bk_01_r(ComponentPose=@n10, Alpha=@n11, ExternalForce="0,0,0") {
    angularSpringConstant=50
    bAngularSpring=true
    bUseGravityOverride=true
    simulationSpace=RootRelative
    boundBone {boneName=dyn_shorts_bk_01_r}
    componentAppliedLinearAccClamp=<1500 1500 1500>
    componentLinearAccScale=<0.2 0.2 0.3>
    componentLinearVelScale=<0.2 0.2 0.4>
    gravityOverride {z=-980}
    physicsBodyDefinitions [{
      boundBone {boneName=dyn_shorts_bk_01_r}
      boxExtents=<16 16 16>
      constraintSetup {angularLimitsMax=<0 5 5>, angularLimitsMin=<0 -5 -30>, angularTarget=<0 0 0>}
      localJointOffset=<1 0 0>
     }]
    simSpaceSettings {maxAngularAcceleration=15, maxAngularVelocity=3, simSpaceAngularAlpha=0.1}
   }
   n13 = Anim Dynamics - Bone: dyn_shorts_bk_02_r(ComponentPose=@n12, Alpha=@n11, ExternalForce="0,0,0") {
    angularSpringConstant=50
    bAngularSpring=true
    bUseGravityOverride=true
    simulationSpace=RootRelative
    boundBone {boneName=dyn_shorts_bk_02_r}
    componentAppliedLinearAccClamp=<1500 1500 1500>
    componentLinearAccScale=<0.2 0.2 0.3>
    componentLinearVelScale=<0.2 0.2 0.4>
    gravityOverride {z=-980}
    physicsBodyDefinitions [{
      boundBone {boneName=dyn_shorts_bk_02_r}
      boxExtents=<16 16 16>
      constraintSetup {angularLimitsMax=<0 5 0>, angularLimitsMin=<0 -5 -30>, angularTarget=<0 0 0>}
      localJointOffset=<1 0 0>
     }]
    simSpaceSettings {maxAngularAcceleration=15, maxAngularVelocity=3, simSpaceAngularAlpha=0.1}
   }
   n14 = Anim Dynamics - Bone: dyn_shorts_out_01_r(ComponentPose=@n13, Alpha=@n11, ExternalForce="0,0,0") {
    angularSpringConstant=50
    bAngularSpring=true
    bUseGravityOverride=true
    simulationSpace=RootRelative
    boundBone {boneName=dyn_shorts_out_01_r}
    componentAppliedLinearAccClamp=<1500 1500 1500>
    componentLinearAccScale=<0.2 0.2 0.3>
    componentLinearVelScale=<0.2 0.2 0.4>
    gravityOverride {z=-980}
    physicsBodyDefinitions [{
      boundBone {boneName=dyn_shorts_out_01_r}
      boxExtents=<16 16 16>
      constraintSetup {angularLimitsMax=<0 5 2>, angularLimitsMin=<0 -5 -15>, angularTarget=<0 0 0>}
      localJointOffset=<1 0 0>
     }]
    simSpaceSettings {maxAngularAcceleration=15, maxAngularVelocity=3, simSpaceAngularAlpha=0.1}
   }
   n15 = Anim Dynamics - Bone: dyn_shorts_out_02_r(ComponentPose=@n14, Alpha=@n11, ExternalForce="0,0,0") {
    angularSpringConstant=50
    bAngularSpring=true
    bUseGravityOverride=true
    simulationSpace=RootRelative
    boundBone {boneName=dyn_shorts_out_02_r}
    componentAppliedLinearAccClamp=<1500 1500 1500>
    componentLinearAccScale=<0.2 0.2 0.3>
    componentLinearVelScale=<0.2 0.2 0.4>
    gravityOverride {z=-980}
    physicsBodyDefinitions [{
      boundBone {boneName=dyn_shorts_out_02_r}
      boxExtents=<16 16 16>
      constraintSetup {angularLimitsMax=<0 5 2>, angularLimitsMin=<0 -5 -15>, angularTarget=<0 0 0>}
      localJointOffset=<1 0 0>
     }]
    simSpaceSettings {maxAngularAcceleration=15, maxAngularVelocity=3, simSpaceAngularAlpha=0.1}
   }
   n16 = Anim Dynamics - Bone: dyn_shorts_in_01_r(ComponentPose=@n15, Alpha=@n11, ExternalForce="0,0,0") {
    angularSpringConstant=50
    bAngularSpring=true
    bUseGravityOverride=true
    simulationSpace=RootRelative
    boundBone {boneName=dyn_shorts_in_01_r}
    componentAppliedLinearAccClamp=<1500 1500 1500>
    componentLinearAccScale=<0.2 0.2 0.3>
    componentLinearVelScale=<0.2 0.2 0.4>
    gravityOverride {z=-980}
    physicsBodyDefinitions [{
      boundBone {boneName=dyn_shorts_in_01_r}
      boxExtents=<16 16 16>
      constraintSetup {angularLimitsMax=<0 5 2>, angularLimitsMin=<0 -5 -5>, angularTarget=<0 0 0>}
      localJointOffset=<1 0 0>
     }]
    simSpaceSettings {maxAngularAcceleration=15, maxAngularVelocity=3, simSpaceAngularAlpha=0.1}
   }
   n17 = Anim Dynamics - Bone: dyn_shorts_in_02_r(ComponentPose=@n16, Alpha=@n11, ExternalForce="0,0,0") {
    angularSpringConstant=50
    bAngularSpring=true
    bUseGravityOverride=true
    simulationSpace=RootRelative
    boundBone {boneName=dyn_shorts_in_02_r}
    componentAppliedLinearAccClamp=<1500 1500 1500>
    componentLinearAccScale=<0.2 0.2 0.3>
    componentLinearVelScale=<0.2 0.2 0.4>
    gravityOverride {z=-980}
    physicsBodyDefinitions [{
      boundBone {boneName=dyn_shorts_in_02_r}
      boxExtents=<16 16 16>
      constraintSetup {angularLimitsMax=<0 5 2>, angularLimitsMin=<0 -5 -7>, angularTarget=<0 0 0>}
      localJointOffset=<1 0 0>
     }]
    simSpaceSettings {maxAngularAcceleration=15, maxAngularVelocity=3, simSpaceAngularAlpha=0.1}
   }
   n18 = Anim Dynamics - Bone: dyn_shorts_fr_01_r(ComponentPose=@n17, Alpha=@n11, ExternalForce="0,0,0") {
    angularSpringConstant=50
    bAngularSpring=true
    bUseGravityOverride=true
    simulationSpace=RootRelative
    boundBone {boneName=dyn_shorts_fr_01_r}
    componentAppliedLinearAccClamp=<1500 1500 1500>
    componentLinearAccScale=<0.2 0.2 0.3>
    componentLinearVelScale=<0.2 0.2 0.4>
    gravityOverride {z=-980}
    physicsBodyDefinitions [{
      boundBone {boneName=dyn_shorts_fr_01_r}
      boxExtents=<16 16 16>
      constraintSetup {angularLimitsMax=<0 5 5>, angularLimitsMin=<0 -5 -5>, angularTarget=<0 0 0>}
      localJointOffset=<1 0 0>
     }]
    simSpaceSettings {maxAngularAcceleration=15, maxAngularVelocity=3, simSpaceAngularAlpha=0.1}
   }
   n19 = Anim Dynamics - Bone: dyn_shorts_fr_02_r(ComponentPose=@n18, Alpha=@n11, ExternalForce="0,0,0") {
    angularSpringConstant=50
    bAngularSpring=true
    bUseGravityOverride=true
    simulationSpace=RootRelative
    boundBone {boneName=dyn_shorts_fr_02_r}
    componentAppliedLinearAccClamp=<1500 1500 1500>
    componentLinearAccScale=<0.2 0.2 0.3>
    componentLinearVelScale=<0.2 0.2 0.4>
    gravityOverride {z=-980}
    physicsBodyDefinitions [{
      boundBone {boneName=dyn_shorts_fr_02_r}
      boxExtents=<16 16 16>
      constraintSetup {angularLimitsMax=<0 5 5>, angularLimitsMin=<0 -5 -5>, angularTarget=<0 0 0>}
      localJointOffset=<1 0 0>
     }]
    simSpaceSettings {maxAngularAcceleration=15, maxAngularVelocity=3, simSpaceAngularAlpha=0.1}
   }
   n20 = "Constraint - MSTR_chain_l (PPPP)"!(
    ComponentPose=@n19
    Alpha=1
    ConstraintWeights_0=0.063
    ConstraintWeights_1:float=0.5197
    ConstraintWeights_2:float=0.2551
    ConstraintWeights_3:float=0.1622
   ) -> (Pose) {
    boneToModify {boneName=MSTR_chain_l}
    constraintSetup [{targetBone {boneName=spine_04}}, {targetBone {boneName=spine_05}}, {targetBone {boneName=neck_01}}, {targetBone {boneName=clavicle_l}}]
    constraintWeights [1,1,1,1]
   }
   n21 = "Constraint - MSTR_chain_r (PPPP)"!(
    ComponentPose=@n20
    Alpha=1
    ConstraintWeights_0=0.063
    ConstraintWeights_1:float=0.519
    ConstraintWeights_2:float=0.255
    ConstraintWeights_3:float=0.163
   ) -> (Pose) {
    boneToModify {boneName=MSTR_chain_r}
    constraintSetup [{targetBone {boneName=spine_04}}, {targetBone {boneName=spine_05}}, {targetBone {boneName=neck_01}}, {targetBone {boneName=clavicle_r}}]
    constraintWeights [1,1,1,1]
   }
   n22 = Look At - Bone: MSTR_chain_l(ComponentPose=@n21, Alpha=1) {
    bUseLookUpAxis=true
    boneToModify {boneName=MSTR_chain_l}
    lookAtLocation=<5.0817 10.8838 0.0889>
    lookAtTarget {boneReference {boneName=spine_05}}
    lookAt_Axis {axis=<1 0>}
    lookUp_Axis {axis {x=0, z=1}}
   }
   n23 = Look At - Bone: MSTR_chain_r(ComponentPose=@n22, Alpha=1) {
    bUseLookUpAxis=true
    boneToModify {boneName=MSTR_chain_r}
    lookAtLocation=<5.0817 10.6787 0.0889>
    lookAtTarget {boneReference {boneName=spine_05}}
    lookAt_Axis {axis=<1 0>}
    lookUp_Axis {axis {x=0, z=1}}
   }
   n24 = Anim Dynamics - Bone: dyn_chain_01_l(ComponentPose=@n23, Alpha=1, ExternalForce="0,0,0") {
    angularSpringConstant=200
    bAngularSpring=true
    bUseGravityOverride=true
    simulationSpace=RootRelative
    boundBone {boneName=dyn_chain_01_l}
    componentAppliedLinearAccClamp=<3000 3000 3000>
    componentLinearAccScale=<0.2 0.2 0.2>
    componentLinearVelScale=<0.2 0.2 0.5>
    gravityOverride {z=-980}
    physicsBodyDefinitions [{
      boundBone {boneName=dyn_chain_01_l}
      boxExtents=<12 12 12>
      constraintSetup {angularLimitsMax=<0 0 0>, angularLimitsMin=<0 -5 -20>, angularTarget=<1 0 0>}
      localJointOffset=<1 0 0>
     }]
    relativeSpaceBone {boneName=spine_05}
    simSpaceSettings {maxAngularAcceleration=3000, maxAngularVelocity=3000, simSpaceAngularAlpha=0.2}
   }
   n25 = Anim Dynamics - Bone: dyn_chain_02_l(ComponentPose=@n24, Alpha=1, ExternalForce="0,0,0") {
    angularSpringConstant=200
    bAngularSpring=true
    bUseGravityOverride=true
    simulationSpace=RootRelative
    boundBone {boneName=dyn_chain_02_l}
    componentAppliedLinearAccClamp=<3000 3000 3000>
    componentLinearAccScale=<0.2 0.2 0.2>
    componentLinearVelScale=<0.2 0.2 0.5>
    gravityOverride {z=-980}
    physicsBodyDefinitions [{
      boundBone {boneName=dyn_chain_02_l}
      boxExtents=<12 12 12>
      constraintSetup {angularLimitsMax=<0 0 0>, angularLimitsMin=<0 -5 -30>, angularTarget=<1 0 0>}
      localJointOffset=<1 0 0>
     }]
    relativeSpaceBone {boneName=spine_05}
    simSpaceSettings {maxAngularAcceleration=3000, maxAngularVelocity=3000, simSpaceAngularAlpha=0.2}
   }
   n26 = Anim Dynamics - Bone: dyn_chain_03_l(ComponentPose=@n25, Alpha=1, ExternalForce="0,0,0") {
    angularSpringConstant=200
    bAngularSpring=true
    bUseGravityOverride=true
    simulationSpace=RootRelative
    boundBone {boneName=dyn_chain_03_l}
    componentAppliedLinearAccClamp=<3000 3000 3000>
    componentLinearAccScale=<0.2 0.2 0.2>
    componentLinearVelScale=<0.2 0.2 0.5>
    gravityOverride {z=-980}
    physicsBodyDefinitions [{
      boundBone {boneName=dyn_chain_03_l}
      boxExtents=<12 12 12>
      constraintSetup {angularLimitsMax=<0 5 0>, angularLimitsMin=<0 -5 -10>, angularTarget=<1 0 0>}
      localJointOffset=<1 0 0>
     }]
    relativeSpaceBone {boneName=spine_05}
    simSpaceSettings {maxAngularAcceleration=3000, maxAngularVelocity=3000, simSpaceAngularAlpha=0.2}
   }
   n27 = Anim Dynamics - Bone: dyn_chain_01_r(ComponentPose=@n26, Alpha=1, ExternalForce="0,0,0") {
    angularSpringConstant=200
    bAngularSpring=true
    bUseGravityOverride=true
    simulationSpace=RootRelative
    boundBone {boneName=dyn_chain_01_r}
    componentAppliedLinearAccClamp=<3000 3000 3000>
    componentLinearAccScale=<0.2 0.2 0.2>
    componentLinearVelScale=<0.2 0.2 0.5>
    gravityOverride {z=-980}
    physicsBodyDefinitions [{
      boundBone {boneName=dyn_chain_01_r}
      boxExtents=<12 12 12>
      constraintSetup {angularLimitsMax=<0 0 0>, angularLimitsMin=<0 -5 -20>, angularTarget=<1 0 0>}
      localJointOffset=<1 0 0>
     }]
    relativeSpaceBone {boneName=spine_05}
    simSpaceSettings {maxAngularAcceleration=3000, maxAngularVelocity=3000, simSpaceAngularAlpha=0.2}
   }
   n28 = Anim Dynamics - Bone: dyn_chain_02_r(ComponentPose=@n27, Alpha=1, ExternalForce="0,0,0") {
    angularSpringConstant=200
    bAngularSpring=true
    bUseGravityOverride=true
    simulationSpace=RootRelative
    boundBone {boneName=dyn_chain_02_r}
    componentAppliedLinearAccClamp=<3000 3000 3000>
    componentLinearAccScale=<0.2 0.2 0.2>
    componentLinearVelScale=<0.2 0.2 0.5>
    gravityOverride {z=-980}
    physicsBodyDefinitions [{
      boundBone {boneName=dyn_chain_02_r}
      boxExtents=<12 12 12>
      constraintSetup {angularLimitsMax=<0 0 0>, angularLimitsMin=<0 -5 -30>, angularTarget=<1 0 0>}
      localJointOffset=<1 0 0>
     }]
    relativeSpaceBone {boneName=spine_05}
    simSpaceSettings {maxAngularAcceleration=3000, maxAngularVelocity=3000, simSpaceAngularAlpha=0.2}
   }
   n29 = Anim Dynamics - Bone: dyn_chain_03_r(ComponentPose=@n28, Alpha=1, ExternalForce="0,0,0") {
    angularSpringConstant=200
    bAngularSpring=true
    bUseGravityOverride=true
    simulationSpace=RootRelative
    boundBone {boneName=dyn_chain_03_r}
    componentAppliedLinearAccClamp=<3000 3000 3000>
    componentLinearAccScale=<0.2 0.2 0.2>
    componentLinearVelScale=<0.2 0.2 0.5>
    gravityOverride {z=-980}
    physicsBodyDefinitions [{
      boundBone {boneName=dyn_chain_03_r}
      boxExtents=<12 12 12>
      constraintSetup {angularLimitsMax=<0 5 0>, angularLimitsMin=<0 -5 -10>, angularTarget=<1 0 0>}
      localJointOffset=<1 0 0>
     }]
    relativeSpaceBone {boneName=spine_05}
    simSpaceSettings {maxAngularAcceleration=3000, maxAngularVelocity=3000, simSpaceAngularAlpha=0.2}
   }
   Output Pose(Result=Component To Local(ComponentPose=@n29)) {layerGroup=None}
  ]}
 EventGraph {nodes [
   Event BlueprintUpdateAnimation()
   n1 = Self-Reference()
   > Set KneeBendL(KneeBendL:float=GetAlphaFromReferenceTransformDelta(
     SkeletalMeshComponent=GetPartSkeletalMeshComponent(self=@n1)
     BoneName=calf_l
     BaseName=thigh_l
     InputRange=(X=75,Y=120)
     OutputRange=(X=1,Y=0)
     TwistAxis="1,0,0"
    )) -> (Output_Get:float)
   n2 = Self-Reference()
   > Set KneeBendR(KneeBendR:float=GetAlphaFromReferenceTransformDelta(
     SkeletalMeshComponent=GetPartSkeletalMeshComponent(self=@n2)
     BoneName=calf_r
     BaseName=thigh_r
     InputRange=(X=75,Y=120)
     OutputRange=(X=1,Y=0)
     TwistAxis="1,0,0"
    )) -> (Output_Get:float)
  ]}
]
variables [
 KneeBendL {type=number}
 KneeBendR {type=number}
]
