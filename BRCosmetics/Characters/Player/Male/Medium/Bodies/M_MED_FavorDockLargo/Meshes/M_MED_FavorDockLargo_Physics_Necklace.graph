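# Physics asset for the FavorDockLargo medium male player body, named for its necklace, though it holds only body bones and no necklace bones.
# One capsule per body approximates pelvis, spine, clavicles, arms, hands, thighs and calves, using twist bones for upper arms and thighs, with no head, neck or foot bodies.
# Constraints with uniform limit angles form a tree from the pelvis, with forearms and upper-arm twists linked to clavicles, and calves and thigh twists directly to pelvis.
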PhysicsAsset /BRCosmetics/Characters/Player/Male/Medium/Bodies/M_MED_FavorDockLargo/Meshes/M_MED_FavorDockLargo_Physics_Necklace
bodies {
 calf_l [{shapeType=Capsule, length=28.3694, radius=12.1157, center=<-26.884 1.383 -0.3789>, rotation=<rot 0 -106.9186 -89.8061>}]
 calf_r [{shapeType=Capsule, length=28.354, radius=12.1265, center=<26.8382 -1.4611 0.3761>, rotation=<rot 0 -106.703 -89.7974>}]
 clavicle_l [{shapeType=Capsule, length=19.3248, radius=8.9454, center=<21.1976 0.1651 -7.4445>, rotation=<rot 0 -85.1581 -53.1071>}]
 clavicle_r [{shapeType=Capsule, length=19.3248, radius=8.9454, center=<-21.1976 -0.1651 7.4445>, rotation=<rot 0 -85.1581 -53.1071>}]
 hand_l [{shapeType=Capsule, length=8.8422, radius=6.3003, center=<7.8205 -2.7207 0.0923>, rotation=<rot 0 83.9087 -68.5025>}]
 hand_r [{shapeType=Capsule, length=8.8442, radius=6.2909, center=<-7.8226 2.7183 -0.094>, rotation=<rot 0 83.8117 -68.6053>}]
 lowerarm_l [{shapeType=Capsule, length=13.064, radius=5.2113, center=<10.7842 0.2436 1.1631>, rotation=<rot 0 92.5408 -83.8599>}]
 lowerarm_r [{shapeType=Capsule, length=13.0799, radius=5.2028, center=<-10.815 -0.4283 -0.398>, rotation=<rot 0 92.6209 -89.1968>}]
 pelvis [{shapeType=Capsule, length=28.9678, radius=17.2203, center=<-13.6622 -0.8874 0.2437>, rotation=<rot 0 -93.6075 -89.8279>}]
 spine_01 [{shapeType=Capsule, length=11.9841, radius=8.935, center=<8.9924 0.4313 -0.0018>, rotation=<rot -0.2541 -90 0.3117>}]
 spine_03 [{shapeType=Capsule, length=15.751, radius=11.7471, center=<5.3972 0.885 0.0417>, rotation=<rot -0.4658 -90 0.1918>}]
 spine_04 [{shapeType=Capsule, length=17.1544, radius=11.9098, center=<6.7855 0.623 -0.0012>, rotation=<rot 0.0146 -90 -0.0029>}]
 spine_05 [{shapeType=Capsule, length=11.1607, radius=9.937, center=<9.527 -0.1958 0.0082>, rotation=<rot -0.2958 -90 0.2017>}]
 thigh_twist_01_l [{shapeType=Capsule, length=17.7979, radius=10.7677, center=<2.4853 -1.0869 0.1059>, rotation=<rot 0 -88.3669 -85.8753>}]
 thigh_twist_01_r [{shapeType=Capsule, length=17.8403, radius=10.9438, center=<-2.4726 1.2719 -0.5293>, rotation=<rot 0 -88.3587 -87.9745>}]
 upperarm_twist_01_l [{shapeType=Capsule, length=8.7214, radius=8.47, center=<2.9326 -0.816 -1.1921>, rotation=<rot 0 -142.1777 -57.4355>}]
 upperarm_twist_01_r [{shapeType=Capsule, length=8.7213, radius=8.47, center=<-2.9326 0.816 1.1921>, rotation=<rot 0 -142.1778 -57.435>}]
}
constraints [
 {bone1Name=spine_01, bone2Name=pelvis, swing1LimitDegrees=45, swing2LimitDegrees=45, twistLimitDegrees=45}
 {bone1Name=spine_03, bone2Name=spine_01, swing1LimitDegrees=45, swing2LimitDegrees=45, twistLimitDegrees=45}
 {bone1Name=spine_04, bone2Name=spine_03, swing1LimitDegrees=45, swing2LimitDegrees=45, twistLimitDegrees=45}
 {bone1Name=spine_05, bone2Name=spine_04, swing1LimitDegrees=45, swing2LimitDegrees=45, twistLimitDegrees=45}
 {bone1Name=clavicle_l, bone2Name=spine_05, swing1LimitDegrees=45, swing2LimitDegrees=45, twistLimitDegrees=45}
 {bone1Name=lowerarm_l, bone2Name=clavicle_l, swing1LimitDegrees=45, swing2LimitDegrees=45, twistLimitDegrees=45}
 {bone1Name=hand_l, bone2Name=lowerarm_l, swing1LimitDegrees=45, swing2LimitDegrees=45, twistLimitDegrees=45}
 {bone1Name=upperarm_twist_01_l, bone2Name=clavicle_l, swing1LimitDegrees=45, swing2LimitDegrees=45, twistLimitDegrees=45}
 {bone1Name=clavicle_r, bone2Name=spine_05, swing1LimitDegrees=45, swing2LimitDegrees=45, twistLimitDegrees=45}
 {bone1Name=lowerarm_r, bone2Name=clavicle_r, swing1LimitDegrees=45, swing2LimitDegrees=45, twistLimitDegrees=45}
 {bone1Name=hand_r, bone2Name=lowerarm_r, swing1LimitDegrees=45, swing2LimitDegrees=45, twistLimitDegrees=45}
 {bone1Name=upperarm_twist_01_r, bone2Name=clavicle_r, swing1LimitDegrees=45, swing2LimitDegrees=45, twistLimitDegrees=45}
 {bone1Name=calf_l, bone2Name=pelvis, swing1LimitDegrees=45, swing2LimitDegrees=45, twistLimitDegrees=45}
 {bone1Name=thigh_twist_01_l, bone2Name=pelvis, swing1LimitDegrees=45, swing2LimitDegrees=45, twistLimitDegrees=45}
 {bone1Name=calf_r, bone2Name=pelvis, swing1LimitDegrees=45, swing2LimitDegrees=45, twistLimitDegrees=45}
 {bone1Name=thigh_twist_01_r, bone2Name=pelvis, swing1LimitDegrees=45, swing2LimitDegrees=45, twistLimitDegrees=45}
]
